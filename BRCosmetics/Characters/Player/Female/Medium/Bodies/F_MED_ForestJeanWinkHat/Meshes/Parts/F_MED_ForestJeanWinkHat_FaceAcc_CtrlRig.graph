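# Control Rig for the F_MED_ForestJeanWinkHat face accessory, adjusting its three hair chains and hat to follow body poses.
# Parent constraints blend hair control nulls across head, neck, spine, clavicle and upper-arm bones.
# Head-rotation remaps and spherical pose readers weight corrective offsets, while twist and B-spline IK functions pose the hair bones.
# A final hat pass applies head-driven correctives, with interpolated crouch and skydive flags scaling some offsets.
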ControlRigBlueprint /BRCosmetics/Characters/Player/Female/Medium/Bodies/F_MED_ForestJeanWinkHat/Meshes/Parts/F_MED_ForestJeanWinkHat_FaceAcc_CtrlRig
graphs [
 Multiply_2_ContainedGraph {nodes [
   n1 = RigVMFunctionEntryNode()
   RigVMFunctionReturnNode(Result=&RigVMFunction_MathDoubleMul(B=@n1.C, A=&RigVMFunction_MathDoubleMul(A=@n1.A, B=@n1.B).Result).Result)
  ]}
 New Function_ContainedGraph {nodes [
   RigVMFunctionEntryNode()
   > &RigUnit_ParentConstraint(
    AdvancedSettings=(RotationOrderForFilter=XZY)
    Child=(Type=Null,Name="dyn_hair_1_l_ctrl_null")
    Parents=((Item=(Type=Bone,Name="head"),Weight=1))
    Weight=1
   )
   > &RigUnit_ParentConstraint(
    AdvancedSettings=(RotationOrderForFilter=XZY)
    Child=(Type=Null,Name="dyn_hair_1_c_ctrl_null")
    Parents=((Item=(Type=Bone,Name="head"),Weight=1))
    Weight=1
   )
   > &RigUnit_ParentConstraint(
    AdvancedSettings=(RotationOrderForFilter=XZY)
    Child=(Type=Null,Name="dyn_hair_1_r_ctrl_null")
    Parents=((Item=(Type=Bone,Name="head"),Weight=1))
    Weight=1
   )
   > &RigUnit_ParentConstraint(
    AdvancedSettings=(RotationOrderForFilter=XZY)
    Child=(Type=Null,Name="dyn_hair_2_l_ctrl_null")
    Filter=(TranslationFilter=(bX=True,bY=True,bZ=True),RotationFilter=(bX=true,bY=True,bZ=True),ScaleFilter=(bX=True,bY=True,bZ=True))
    Parents=((Item=(Type=Bone,Name="head"),Weight=0.1107),(Item=(Type=Bone,Name="clavicle_l"),Weight=0.0212),(Item=(Type=Bone,Name="neck_02"),Weight=0.2695),(Item=(Type=Bone,Name="spine_05"),Weight=0.2165),(Item=(Type=Bone,Name="neck_01"),Weight=0.3821))
    Weight=1
   )
   > &RigUnit_ParentConstraint(
    AdvancedSettings=(RotationOrderForFilter=XZY)
    Child=(Type=Null,Name="dyn_hair_2_c_ctrl_null")
    Parents=((Item=(Type=Bone,Name="head"),Weight=0.1532),(Item=(Type=Bone,Name="spine_05"),Weight=0.0348),(Item=(Type=Bone,Name="neck_02"),Weight=0.4018),(Item=(Type=Bone,Name="neck_01"),Weight=0.4103))
    Weight=1
   )
   > &RigUnit_ParentConstraint(
    AdvancedSettings=(RotationOrderForFilter=XZY)
    Child=(Type=Null,Name="dyn_hair_2_r_ctrl_null")
    Filter=(TranslationFilter=(bX=True,bY=True,bZ=True),RotationFilter=(bX=true,bY=True,bZ=True),ScaleFilter=(bX=True,bY=True,bZ=True))
    Parents=((Item=(Type=Bone,Name="head"),Weight=0.1107),(Item=(Type=Bone,Name="clavicle_r"),Weight=0.0212),(Item=(Type=Bone,Name="neck_02"),Weight=0.2695),(Item=(Type=Bone,Name="spine_05"),Weight=0.2165),(Item=(Type=Bone,Name="neck_01"),Weight=0.3821))
    Weight=1
   )
   > &RigUnit_ParentConstraint(
    AdvancedSettings=(RotationOrderForFilter=XZY)
    Child=(Type=Null,Name="dyn_hair_3_r_ctrl_null")
    Parents=((Item=(Type=Bone,Name="spine_04"),Weight=0.0026),(Item=(Type=Bone,Name="upperarm_r"),Weight=0.0564),(Item=(Type=Bone,Name="spine_05"),Weight=0.2326),(Item=(Type=Bone,Name="clavicle_r"),Weight=0.7084))
    Weight=1
   )
   > &RigUnit_ParentConstraint(
    AdvancedSettings=(RotationOrderForFilter=XZY)
    Child=(Type=Null,Name="dyn_hair_4_r_ctrl_null")
    Parents=((Item=(Type=Bone,Name="upperarm_r"),Weight=0.0788),(Item=(Type=Bone,Name="spine_04"),Weight=0.0853),(Item=(Type=Bone,Name="spine_05"),Weight=0.2295),(Item=(Type=Bone,Name="clavicle_r"),Weight=0.6064))
    Weight=1
   )
   > &RigUnit_ParentConstraint(
    AdvancedSettings=(RotationOrderForFilter=XZY)
    Child=(Type=Null,Name="dyn_hair_5_r_ctrl_null")
    Parents="((Item=(Type=Bone,Name=\"clavicle_l\"),Weight=0.009),(Item=(Type=Bone,Name=\"spine_02\"),Weight=0.0315),(Item=(Type=Bone,Name=\"spine_03\"),Weight=0.0319),(Item=(Type=Bone,Name=\"upperarm_r\"),Weight=0.0646),(Item=(Type=Bone,Name=\"spine_05\"),Weight=0.1926),(Item=( ...(truncated)"
    Weight=1
   )
   > &RigUnit_ParentConstraint(
    AdvancedSettings=(RotationOrderForFilter=XZY)
    Child=(Type=Null,Name="dyn_hair_3_l_ctrl_null")
    Parents=((Item=(Type=Bone,Name="spine_04"),Weight=0.0026),(Item=(Type=Bone,Name="upperarm_l"),Weight=0.0564),(Item=(Type=Bone,Name="spine_05"),Weight=0.2326),(Item=(Type=Bone,Name="clavicle_l"),Weight=0.7084))
    Weight=1
   )
   > &RigUnit_ParentConstraint(
    AdvancedSettings=(RotationOrderForFilter=XZY)
    Child=(Type=Null,Name="dyn_hair_4_l_ctrl_null")
    Parents=((Item=(Type=Bone,Name="upperarm_l"),Weight=0.0788),(Item=(Type=Bone,Name="spine_04"),Weight=0.0853),(Item=(Type=Bone,Name="spine_05"),Weight=0.2295),(Item=(Type=Bone,Name="clavicle_l"),Weight=0.6064))
    Weight=1
   )
   > &RigUnit_ParentConstraint(
    AdvancedSettings=(RotationOrderForFilter=XZY)
    Child=(Type=Null,Name="dyn_hair_5_l_ctrl_null")
    Parents="((Item=(Type=Bone,Name=\"clavicle_r\"),Weight=0.009),(Item=(Type=Bone,Name=\"spine_02\"),Weight=0.0315),(Item=(Type=Bone,Name=\"spine_03\"),Weight=0.0319),(Item=(Type=Bone,Name=\"upperarm_l\"),Weight=0.0646),(Item=(Type=Bone,Name=\"spine_05\"),Weight=0.1926),(Item=( ...(truncated)"
    Weight=1
   )
   > &RigUnit_ParentConstraint(
    AdvancedSettings=(RotationOrderForFilter=XZY)
    Child=(Type=Null,Name="dyn_hair_3_c_ctrl_null")
    Parents=((Item=(Type=Bone,Name="clavicle_r"),Weight=0.068),(Item=(Type=Bone,Name="clavicle_l"),Weight=0.0703),(Item=(Type=Bone,Name="spine_05"),Weight=0.8616))
    Weight=1
   )
   > &RigUnit_ParentConstraint(
    AdvancedSettings=(RotationOrderForFilter=XZY)
    Child=(Type=Null,Name="dyn_hair_4_c_ctrl_null")
    Parents=((Item=(Type=Bone,Name="spine_04"),Weight=0.0314),(Item=(Type=Bone,Name="clavicle_l"),Weight=0.0891),(Item=(Type=Bone,Name="clavicle_r"),Weight=0.279),(Item=(Type=Bone,Name="spine_05"),Weight=0.6005))
    Weight=1
   )
   > &RigUnit_ParentConstraint(
    AdvancedSettings=(RotationOrderForFilter=XZY)
    Child=(Type=Null,Name="dyn_hair_5_c_ctrl_null")
    Parents="((Item=(Type=Bone,Name=\"spine_01\"),Weight=0.0186),(Item=(Type=Bone,Name=\"spine_03\"),Weight=0.0604),(Item=(Type=Bone,Name=\"spine_02\"),Weight=0.0617),(Item=(Type=Bone,Name=\"clavicle_l\"),Weight=0.0658),(Item=(Type=Bone,Name=\"clavicle_r\"),Weight=0.1694),(Item= ...(truncated)"
    Weight=1
   )
   > RigVMFunctionReturnNode()
  ]}
 New Function_ContainedGraph {nodes [
   RigVMFunctionEntryNode()
   > RigVMFunction_Sequence = RigVMAggregateNode()
   DeltaRotation v4_3 = RigVMFunctionReferenceNode(
    Rotation Order=ZYX
    child=(Type=Bone,Name="head")
    debugRotation=false
    onlyTwist=true
    outputMinMax=(X=0,Y=1)
    parent=(Type=Null,Name="head_null")
    premult=1
    rotationMinMax=(X=0,Y=-60)
    twistAxis=(X=1,Y=0,Z=0)
    ExecuteContext=@RigVMFunction_Sequence.A
   )
   > &RigUnit_ModifyTransforms(
    ItemToModify="((Item=(Type=Null,Name=\"dyn_hair_2_r_ctrl_null\"),Transform=(Rotation=(X=0,Y=0,Z=0,W=1),Translation=(X=0,Y=0,Z=0),Scale3D=(X=1,Y=1,Z=1))),(Item=(Type=Null,Name=\"dyn_hair_3_r_ctrl_null\"),Transform=(Rotation=(X=0,Y=0,Z=0,W=-1),Translation=(X=0,Y=-3,Z=3),Scale ...(truncated)"
    WeightMinimum=0
    Weight=@"DeltaRotation v4_3".outputWeight
   )
   > RigVMVariableNode(Value=@"DeltaRotation v4_3".outputWeight) {variable=HeadRotLeft}
   > DeltaRotation v4_5 = RigVMFunctionReferenceNode(
    Rotation Order=ZYX
    child=(Type=Bone,Name="head")
    debugRotation=false
    onlyTwist=true
    outputMinMax=(X=0,Y=1)
    parent=(Type=Null,Name="head_null")
    premult=1
    rotationMinMax=(X=0,Y=60)
    twistAxis=(X=1,Y=0,Z=0)
   )
   > &RigUnit_ModifyTransforms(
    ItemToModify="((Item=(Type=Null,Name=\"dyn_hair_1_r_ctrl_null\"),Transform=(Rotation=(X=0,Y=0,Z=0,W=1),Translation=(X=-0.2341,Y=-3.6072,Z=-0.9853),Scale3D=(X=1,Y=1,Z=1))),(Item=(Type=Null,Name=\"dyn_hair_2_r_ctrl_null\"),Transform=(Rotation=(X=0,Y=0,Z=0,W=-1),Translation=(X ...(truncated)"
    WeightMinimum=0
    Weight=@"DeltaRotation v4_5".outputWeight
   )
   > DeltaRotation v4_7 = RigVMFunctionReferenceNode(
    Rotation Order=ZYX
    child=(Type=Bone,Name="head")
    debugRotation=false
    outputMinMax=(X=0,Y=1)
    parent=(Type=Null,Name="head_null")
    premult=1
    rotationMinMax=(X=-10,Y=-40)
    triggerAxis=Z
    twistAxis=(X=1,Y=0,Z=0)
   )
   > &RigUnit_ModifyTransforms(
    ItemToModify="((Item=(Type=Null,Name=\"dyn_hair_1_c_ctrl_null\"),Transform=(Rotation=(X=0,Y=0,Z=0,W=1),Translation=(X=-6,Y=0,Z=5),Scale3D=(X=1,Y=1,Z=1))),(Item=(Type=Null,Name=\"dyn_hair_1_l_ctrl_null\"),Transform=(Rotation=(X=0,Y=0,Z=0,W=-1),Translation=(X=-2,Y=0,Z=2),Scal ...(truncated)"
    WeightMinimum=0
    Weight=@"DeltaRotation v4_7".outputWeight
   )
   > DeltaRotation v4_1 = RigVMFunctionReferenceNode(
    Rotation Order=ZYX
    child=(Type=Bone,Name="head")
    debugRotation=false
    outputMinMax=(X=0,Y=1)
    parent=(Type=Null,Name="head_null")
    premult=1
    rotationMinMax=(X=-1,Y=-45)
    triggerAxis=Z
    twistAxis=(X=1,Y=0,Z=0)
   )
   > &RigUnit_ModifyTransforms(
    ItemToModify="((Item=(Type=Null,Name=\"dyn_hair_5_r_ctrl_null\"),Transform=(Rotation=(X=0,Y=0,Z=0,W=1),Translation=(X=5,Y=0,Z=0),Scale3D=(X=1,Y=1,Z=1))),(Item=(Type=Null,Name=\"dyn_hair_5_c_ctrl_null\"),Transform=(Rotation=(X=0,Y=0,Z=0,W=1),Translation=(X=6,Y=0,Z=0),Scale3D ...(truncated)"
    WeightMinimum=0
    Weight=@"DeltaRotation v4_1".outputWeight
   )
   > n1 = &RigUnit_SphericalPoseReader(
    ActiveRegionScaleFactors=(PositiveWidth=1,NegativeWidth=1,PositiveHeight=1,NegativeHeight=1)
    ActiveRegionSize=0.05
    Debug=(bDrawDebug=false,bDraw2D=False,bDrawLocalAxes=False,DebugScale=25,DebugSegments=20,DebugThickness=0.25)
    DriverAxis=(X=1,Y=0,Z=0)
    DriverItem=(Type=Bone,Name="clavicle_l")
    FalloffRegionScaleFactors=(PositiveWidth=1,NegativeWidth=1,PositiveHeight=1,NegativeHeight=1)
    FalloffSize=0.17
    FlipHeightScaling=False
    FlipWidthScaling=False
    OptionalParentItem=(Type=Bone,Name="spine_03")
    RotationOffset=(X=-75,Y=0,Z=75)
   )
   > &RigUnit_ModifyTransforms(
    ItemToModify="((Item=(Type=Bone,Name=\"None\"),Transform=(Rotation=(X=0,Y=0,Z=0,W=1),Translation=(X=0,Y=0,Z=0),Scale3D=(X=1,Y=1,Z=1))),(Item=(Type=Null,Name=\"dyn_hair_3_l_ctrl_null\"),Transform=(Rotation=(X=0,Y=0,Z=0,W=1),Translation=(X=0,Y=0,Z=4),Scale3D=(X=1,Y=1,Z=1))),( ...(truncated)"
    WeightMinimum=0
    Weight=@n1.OutputParam
   )
   > n2 = &RigUnit_SphericalPoseReader(
    ActiveRegionScaleFactors=(PositiveWidth=1,NegativeWidth=1,PositiveHeight=1,NegativeHeight=1)
    ActiveRegionSize=0.05
    Debug=(bDrawDebug=false,bDraw2D=False,bDrawLocalAxes=False,DebugScale=25,DebugSegments=20,DebugThickness=0.25)
    DriverAxis=(X=-1,Y=0,Z=0)
    DriverItem=(Type=Bone,Name="clavicle_r")
    FalloffRegionScaleFactors=(PositiveWidth=1,NegativeWidth=1,PositiveHeight=1,NegativeHeight=1)
    FalloffSize=0.17
    FlipHeightScaling=False
    FlipWidthScaling=False
    OptionalParentItem=(Type=Bone,Name="spine_03")
    RotationOffset=(X=105,Y=0,Z=70)
   )
   > &RigUnit_ModifyTransforms(
    ItemToModify="((Item=(Type=Bone,Name=\"None\"),Transform=(Rotation=(X=0,Y=0,Z=0,W=1),Translation=(X=0,Y=0,Z=0),Scale3D=(X=1,Y=1,Z=1))),(Item=(Type=Null,Name=\"dyn_hair_3_r_ctrl_null\"),Transform=(Rotation=(X=0,Y=0,Z=0,W=1),Translation=(X=0,Y=-4,Z=2),Scale3D=(X=1,Y=1,Z=1))), ...(truncated)"
    WeightMinimum=0
    Weight=@n2.OutputParam
   )
   > n3 = &RigUnit_SphericalPoseReader(
    ActiveRegionScaleFactors=(PositiveWidth=1,NegativeWidth=1,PositiveHeight=1,NegativeHeight=1)
    ActiveRegionSize=0.04
    Debug=(bDrawDebug=false,bDraw2D=False,bDrawLocalAxes=False,DebugScale=25,DebugSegments=20,DebugThickness=0.25)
    DriverAxis=(X=1,Y=0,Z=0)
    DriverItem=(Type=Bone,Name="upperarm_l")
    FalloffRegionScaleFactors=(PositiveWidth=1,NegativeWidth=1,PositiveHeight=1,NegativeHeight=1)
    FalloffSize=0.2
    FlipHeightScaling=False
    FlipWidthScaling=False
    OptionalParentItem=(Type=Bone,Name="spine_04")
    RotationOffset=(X=40,Y=0,Z=-10)
   )
   > &RigUnit_ModifyTransforms(
    ItemToModify="((Item=(Type=Bone,Name=\"None\"),Transform=(Rotation=(X=0,Y=0,Z=0,W=1),Translation=(X=0,Y=0,Z=0),Scale3D=(X=1,Y=1,Z=1))),(Item=(Type=Null,Name=\"dyn_hair_2_l_ctrl_null\"),Transform=(Rotation=(X=0,Y=0,Z=0,W=1),Translation=(X=0,Y=0,Z=5),Scale3D=(X=1,Y=1,Z=1))),( ...(truncated)"
    WeightMinimum=0
    Weight=@n3.OutputParam
   )
   joints_twist_c = RigVMDispatchNode(Value=((Type=Bone,Name="dyn_hair_2_c"),(Type=Bone,Name="dyn_hair_3_c"),(Type=Bone,Name="dyn_hair_4_c"),(Type=Bone,Name="dyn_hair_5_c"))) {template="DISPATCH_RigVMDispatch_Constant(io Value)"}
   TwistBones = RigVMFunctionReferenceNode(
    Axis to Twist=(bX=True,bY=true,bZ=true)
    Debug=false
    End=(Type=Bone,Name="neck_01")
    Start=(Type=Bone,Name="spine_02")
    ExecuteContext=@RigVMFunction_Sequence.B
    Bones to Twist=@joints_twist_c.Value
   )
   joints_twist_l = RigVMDispatchNode(Value=((Type=Bone,Name="dyn_hair_2_l"),(Type=Bone,Name="dyn_hair_3_l"),(Type=Bone,Name="dyn_hair_4_l"),(Type=Bone,Name="dyn_hair_5_l"))) {template="DISPATCH_RigVMDispatch_Constant(io Value)"}
   > TwistBones_1 = RigVMFunctionReferenceNode(
    Axis to Twist=(bX=True,bY=true,bZ=true)
    Debug=false
    End=(Type=Bone,Name="neck_01")
    Start=(Type=Bone,Name="spine_02")
    Bones to Twist=@joints_twist_l.Value
   )
   joints_twist_r = RigVMDispatchNode(Value=((Type=Bone,Name="dyn_hair_2_r"),(Type=Bone,Name="dyn_hair_3_r"),(Type=Bone,Name="dyn_hair_4_r"),(Type=Bone,Name="dyn_hair_5_r"))) {template="DISPATCH_RigVMDispatch_Constant(io Value)"}
   > TwistBones_2 = RigVMFunctionReferenceNode(
    Axis to Twist=(bX=True,bY=true,bZ=true)
    Debug=false
    End=(Type=Bone,Name="neck_01")
    Start=(Type=Bone,Name="spine_02")
    Bones to Twist=@joints_twist_r.Value
   )
   controls_ik_c_x7 = RigVMDispatchNode(Value=((Type=Control,Name="dyn_hair_1_c_ctrl"),(Type=Control,Name="dyn_hair_2_c_ctrl"),(Type=Control,Name="dyn_hair_3_c_ctrl"),(Type=Control,Name="dyn_hair_4_c_ctrl"),(Type=Control,Name="dyn_hair_5_c_ctrl"))) {template="DISPATCH_RigVMDispatch_Constant(io Value)"}
   joints_ik_c_f_x7 = RigVMDispatchNode(Value=((Type=Bone,Name="dyn_hair_1_c"),(Type=Bone,Name="dyn_hair_2_c"),(Type=Bone,Name="dyn_hair_3_c"),(Type=Bone,Name="dyn_hair_4_c"),(Type=Bone,Name="dyn_hair_5_c"))) {template="DISPATCH_RigVMDispatch_Constant(io Value)"}
   > SplineIK_6 = RigVMFunctionReferenceNode(
    Primary Axis=(X=1,Y=0,Z=0)
    Samples Per Segment=1
    Secondary Spline Direction=(X=0,Y=0,Z=0)
    Spline Mode=BSpline
    Stretch=false
    Up Axis=(X=0,Y=0,Z=0)
    debug=false
    Controls=@controls_ik_c_x7.Value
    Bones=@joints_ik_c_f_x7.Value
   )
   controls_ik_l_f_x7 = RigVMDispatchNode(Value=((Type=Control,Name="dyn_hair_1_l_ctrl"),(Type=Control,Name="dyn_hair_2_l_ctrl"),(Type=Control,Name="dyn_hair_3_l_ctrl"),(Type=Control,Name="dyn_hair_4_l_ctrl"),(Type=Control,Name="dyn_hair_5_l_ctrl"))) {template="DISPATCH_RigVMDispatch_Constant(io Value)"}
   joints_ik_l_f_x7 = RigVMDispatchNode(Value=((Type=Bone,Name="dyn_hair_1_l"),(Type=Bone,Name="dyn_hair_2_l"),(Type=Bone,Name="dyn_hair_3_l"),(Type=Bone,Name="dyn_hair_4_l"),(Type=Bone,Name="dyn_hair_5_l"))) {template="DISPATCH_RigVMDispatch_Constant(io Value)"}
   > SplineIK_7 = RigVMFunctionReferenceNode(
    Primary Axis=(X=1,Y=0,Z=0)
    Samples Per Segment=1
    Secondary Spline Direction=(X=0,Y=0,Z=0)
    Spline Mode=BSpline
    Stretch=false
    Up Axis=(X=0,Y=0,Z=0)
    debug=false
    Controls=@controls_ik_l_f_x7.Value
    Bones=@joints_ik_l_f_x7.Value
   )
   controls_ik_r_x7 = RigVMDispatchNode(Value=((Type=Control,Name="dyn_hair_1_r_ctrl"),(Type=Control,Name="dyn_hair_2_r_ctrl"),(Type=Control,Name="dyn_hair_3_r_ctrl"),(Type=Control,Name="dyn_hair_4_r_ctrl"),(Type=Control,Name="dyn_hair_5_r_ctrl"))) {template="DISPATCH_RigVMDispatch_Constant(io Value)"}
   joints_ik_r_f_x7 = RigVMDispatchNode(Value=((Type=Bone,Name="dyn_hair_1_r"),(Type=Bone,Name="dyn_hair_2_r"),(Type=Bone,Name="dyn_hair_3_r"),(Type=Bone,Name="dyn_hair_4_r"),(Type=Bone,Name="dyn_hair_5_r"))) {template="DISPATCH_RigVMDispatch_Constant(io Value)"}
   > SplineIK_8 = RigVMFunctionReferenceNode(
    Primary Axis=(X=1,Y=0,Z=0)
    Samples Per Segment=1
    Secondary Spline Direction=(X=0,Y=0,Z=0)
    Spline Mode=BSpline
    Stretch=false
    Up Axis=(X=0,Y=0,Z=0)
    debug=false
    Controls=@controls_ik_r_x7.Value
    Bones=@joints_ik_r_f_x7.Value
   )
   DeltaRotation v4_4 = RigVMFunctionReferenceNode(
    Rotation Order=ZYX
    child=(Type=Bone,Name="head")
    debugRotation=false
    outputMinMax=(X=0,Y=1)
    parent=(Type=Null,Name="head_null")
    premult=1
    rotationMinMax=(X=5,Y=-40)
    triggerAxis=Z
    twistAxis=(X=1,Y=0,Z=0)
    ExecuteContext=@RigVMFunction_Sequence.C
   )
   > AlterTransform_3 = RigVMFunctionReferenceNode(
    inheritsTransform=false
    transforms="((Item=(Type=Bone,Name=\"dyn_hair_push_l\"),Transform=(Rotation=(X=0,Y=0,Z=0,W=1),Translation=(X=-1,Y=0,Z=1),Scale3D=(X=1,Y=1,Z=1))),(Item=(Type=Bone,Name=\"dyn_hair_push_r\"),Transform=(Rotation=(X=0,Y=0.0872,Z=0,W=0.9962),Translation=(X=-2.5,Y=0,Z=-2),Scale3 ...(truncated)"
    weights=@"DeltaRotation v4_4".outputWeight
   )
   > DeltaRotation v4_2 = RigVMFunctionReferenceNode(
    Rotation Order=ZYX
    child=(Type=Bone,Name="head")
    debugRotation=false
    outputMinMax=(X=0,Y=1)
    parent=(Type=Null,Name="head_null")
    premult=1
    rotationMinMax=(X=0,Y=-20)
    triggerAxis=Z
    twistAxis=(X=1,Y=0,Z=0)
   )
   > AlterTransform_0 = RigVMFunctionReferenceNode(
    inheritsTransform=false
    transforms="((Item=(Type=Bone,Name=\"dyn_hair_1_c\"),Transform=(Rotation=(X=0,Y=0,Z=0,W=1),Translation=(X=-1,Y=0,Z=2),Scale3D=(X=1,Y=1,Z=1))),(Item=(Type=Bone,Name=\"dyn_hair_1_l\"),Transform=(Rotation=(X=0,Y=0.0262,Z=0,W=0.9997),Translation=(X=-3,Y=0,Z=2.5),Scale3D=(X=1, ...(truncated)"
    weights=@"DeltaRotation v4_2".outputWeight
   )
   n4 = &RigVMFunction_AlphaInterp(
    InterpSpeedDecreasing=15
    InterpSpeedIncreasing=15
    Scale=1
    bClampResult=False
    Value=&RigVMFunction_MathBoolToFloat(Value=RigVMVariableNode() {variable=isCrouching_CR}.Value).Result
   )
   Multiply = &RigVMFunction_MathDoubleMul(B=0.6, A=@n4.Result)
   > &RigUnit_ParentConstraint(
    AdvancedSettings=(RotationOrderForFilter=XZY)
    Child=(Type=Bone,Name="dyn_hair_2_r")
    Filter=(TranslationFilter=(bX=True,bY=True,bZ=True),RotationFilter=(bX=true,bY=True,bZ=True),ScaleFilter=(bX=True,bY=True,bZ=True))
    Parents=((Item=(Type=Bone,Name="head"),Weight=0.1107),(Item=(Type=Bone,Name="clavicle_r"),Weight=0.0212),(Item=(Type=Bone,Name="neck_02"),Weight=0.2695),(Item=(Type=Bone,Name="spine_05"),Weight=0.2165),(Item=(Type=Bone,Name="neck_01"),Weight=0.3821))
    bMaintainOffset=true
    Weight=@Multiply.Result
   )
   > &RigUnit_ParentConstraint(
    AdvancedSettings=(RotationOrderForFilter=XZY)
    Child=(Type=Bone,Name="dyn_hair_3_r")
    Filter=(TranslationFilter=(bX=True,bY=True,bZ=True),RotationFilter=(bX=true,bY=True,bZ=True),ScaleFilter=(bX=True,bY=True,bZ=True))
    Parents=((Item=(Type=Bone,Name="spine_04"),Weight=0.0074),(Item=(Type=Bone,Name="upperarm_r"),Weight=0.0741),(Item=(Type=Bone,Name="spine_05"),Weight=0.2271),(Item=(Type=Bone,Name="clavicle_r"),Weight=0.6915))
    bMaintainOffset=true
    Weight=@Multiply.Result
   )
   > &RigUnit_ModifyTransforms(
    ItemToModify="((Item=(Type=Bone,Name=\"dyn_hair_2_r\"),Transform=(Rotation=(X=0,Y=0,Z=0,W=1),Translation=(X=1,Y=0,Z=2),Scale3D=(X=1,Y=1,Z=1))),(Item=(Type=Bone,Name=\"dyn_hair_2_l\"),Transform=(Rotation=(X=0,Y=0,Z=0,W=1),Translation=(X=0,Y=0,Z=3.5),Scale3D=(X=1,Y=1,Z=1))),( ...(truncated)"
    WeightMinimum=0
    Weight=&RigVMFunction_MathDoubleMul(A=@n4.Result, B=@"DeltaRotation v4_4".outputWeight).Result
   )
   > n5 = &RigUnit_SphericalPoseReader(
    ActiveRegionScaleFactors=(PositiveWidth=1,NegativeWidth=1,PositiveHeight=1,NegativeHeight=1)
    ActiveRegionSize=0.05
    Debug=(bDrawDebug=true,bDraw2D=False,bDrawLocalAxes=False,DebugScale=25,DebugSegments=20,DebugThickness=0.25)
    DriverAxis=(X=1,Y=0,Z=0)
    DriverItem=(Type=Bone,Name="clavicle_l")
    FalloffRegionScaleFactors=(PositiveWidth=1,NegativeWidth=1,PositiveHeight=1,NegativeHeight=1)
    FalloffSize=0.17
    FlipHeightScaling=False
    FlipWidthScaling=False
    OptionalParentItem=(Type=Bone,Name="spine_05")
    RotationOffset=(X=-70,Y=0,Z=60)
   )
   n6 = &RigVMFunction_AlphaInterp(
    InterpSpeedDecreasing=15
    InterpSpeedIncreasing=15
    Scale=1
    bClampResult=False
    Value=&RigVMFunction_MathBoolToFloat(Value=RigVMVariableNode() {variable=isSkydiving_CR}.Value).Result
   )
   Remap = &RigVMFunction_MathDoubleRemap(SourceMaximum=1, SourceMinimum=0, TargetMaximum=0, TargetMinimum=0.6, bClamp=False, Value=@n6.Result)
   Multiply_1 = RigVMAggregateNode(B=0.4, C=@Remap.Result, A=@n5.OutputParam)
   > &RigUnit_ModifyTransforms(
    ItemToModify="((Item=(Type=Bone,Name=\"None\"),Transform=(Rotation=(X=0,Y=0,Z=0,W=1),Translation=(X=0,Y=0,Z=0),Scale3D=(X=1,Y=1,Z=1))),(Item=(Type=Bone,Name=\"base_hair_1_c\"),Transform=(Rotation=(X=-0.0371,Y=-0.2816,Z=-0.1252,W=0.9506),Translation=(X=1.5,Y=0,Z=5),Scale3D=( ...(truncated)"
    WeightMinimum=0
    Weight=@Multiply_1.Result
   )
   > &RigUnit_ModifyTransforms(
    ItemToModify=((Item=(Type=Bone,Name="dyn_hair_3_l"),Transform=(Rotation=(X=0,Y=0,Z=0,W=1),Translation=(X=0,Y=0,Z=2),Scale3D=(X=1,Y=1,Z=1))))
    WeightMinimum=0
    Weight=&RigVMFunction_MathDoubleMul(A=@n5.OutputParam, B=RigVMVariableNode() {variable=HeadRotLeft}.Value).Result
   )
   > n7 = &RigUnit_SphericalPoseReader(
    ActiveRegionScaleFactors=(PositiveWidth=1,NegativeWidth=1,PositiveHeight=1,NegativeHeight=1)
    ActiveRegionSize=0.04
    Debug=(bDrawDebug=false,bDraw2D=False,bDrawLocalAxes=False,DebugScale=25,DebugSegments=20,DebugThickness=0.25)
    DriverAxis=(X=-1,Y=0,Z=0)
    DriverItem=(Type=Bone,Name="clavicle_r")
    FalloffRegionScaleFactors=(PositiveWidth=1,NegativeWidth=1,PositiveHeight=1,NegativeHeight=1)
    FalloffSize=0.12
    FlipHeightScaling=False
    FlipWidthScaling=False
    OptionalParentItem=(Type=Bone,Name="spine_04")
    RotationOffset=(X=-120,Y=0,Z=-90)
   )
   > &RigUnit_ModifyTransforms(
    ItemToModify="((Item=(Type=Bone,Name=\"None\"),Transform=(Rotation=(X=0,Y=0,Z=0,W=1),Translation=(X=0,Y=0,Z=0),Scale3D=(X=1,Y=1,Z=1))),(Item=(Type=Bone,Name=\"dyn_hair_2_r\"),Transform=(Rotation=(X=0,Y=0,Z=0,W=1),Translation=(X=-3,Y=0,Z=2),Scale3D=(X=1,Y=1,Z=1))),(Item=(Typ ...(truncated)"
    WeightMinimum=0
    Weight=@n7.OutputParam
   )
   > DeltaRotation v4 = RigVMFunctionReferenceNode(
    Rotation Order=ZYX
    child=(Type=Bone,Name="upperarm_l")
    debugRotation=false
    outputMinMax=(X=0,Y=1)
    parent=(Type=Bone,Name="upperarm_l_Support")
    premult=1
    rotationMinMax=(X=15,Y=-50)
    triggerAxis=Z
    twistAxis=(X=1,Y=0,Z=0)
   )
   > &RigUnit_ModifyTransforms(
    ItemToModify=((Item=(Type=Bone,Name="upperarm_l_Support"),Transform=(Rotation=(X=0,Y=0,Z=-0.0304,W=0.9995),Translation=(X=-4.6777,Y=-5.5784,Z=-5.6819),Scale3D=(X=1,Y=1,Z=1))))
    WeightMinimum=0
    Weight=@"DeltaRotation v4".outputWeight
   )
   RigVMFunctionReturnNode(ExecuteContext=@RigVMFunction_Sequence.D)
  ]}
 New Function_ContainedGraph {nodes [
   n1 = RigVMFunctionEntryNode()
   RigVMFunction_MathBoolToInteger = &RigVMFunction_MathBoolToInteger(Value=@n1.onlyTwist)
   RigVMFunction_MathQuaternionSwingTwist = &RigVMFunction_MathQuaternionSwingTwist(
    twistAxis=@n1.twistAxis
    Input=&RigUnit_GetRelativeTransformForItem(bChildInitial=False, bParentInitial=False, child=@n1.child, parent=@n1.parent).RelativeTransform.Rotation
   )
   n2 = RigVMDispatchNode(
    Values=((),())
    Index=@RigVMFunction_MathBoolToInteger.Result
    Values.0=@RigVMFunction_MathQuaternionSwingTwist.Swing
    Values.1=@RigVMFunction_MathQuaternionSwingTwist.Twist
   ) {template="DISPATCH_RigVMDispatch_SelectInt32(in Index,in Values,out Result)"}
   RigVMFunction_MathQuaternionToEuler = &RigVMFunction_MathQuaternionToEuler(Value=@n2.Result, RotationOrder=@n1.Rotation Order)
   > DISPATCH_RigVMDispatch_Print = RigVMDispatchNode(ScreenColor=(R=1,G=1,B=1,A=1), ScreenDuration=1.5, Enabled=@n1.debugRotation, Value=@RigVMFunction_MathQuaternionToEuler.Result) {template="DISPATCH_RigVMDispatch_Print(in Prefix,in Value,in Enabled,in ScreenDuration,in ScreenColor)"}
   Cast = RigVMDispatchNode(Value=@n1.triggerAxis) {template="DISPATCH_RigVMDispatch_CastEnumToInt(in Value,out Result)"}
   DISPATCH_RigVMDispatch_SelectInt32 = RigVMDispatchNode(
    Values=(0,0,0)
    Index=@Cast.Result
    Values.0=@RigVMFunction_MathQuaternionToEuler.Result.X
    Values.1=@RigVMFunction_MathQuaternionToEuler.Result.Y
    Values.2=@RigVMFunction_MathQuaternionToEuler.Result.Z
   ) {template="DISPATCH_RigVMDispatch_SelectInt32(in Index,in Values,out Result)"}
   Remap = &RigVMFunction_MathDoubleRemap(
    bClamp=False
    Value=@DISPATCH_RigVMDispatch_SelectInt32.Result
    TargetMinimum=@n1.outputMinMax.X
    TargetMaximum=@n1.outputMinMax.Y
    SourceMinimum=@n1.rotationMinMax.X
    SourceMaximum=@n1.rotationMinMax.Y
   )
   Absolute = &RigVMFunction_MathDoubleAbs(Value=@Remap.Result)
   Multiply = &RigVMFunction_MathDoubleMul(
    A=&RigVMFunction_MathDoubleClamp(Value=RigVMDispatchNode(True=@Absolute.Result, Condition=@n1.absoluteValue, False=@Remap.Result) {template="DISPATCH_RigVMDispatch_If(in Condition,in True,in False,out Result)"}.Result).Result
    B=@n1.premult
   )
   > RigVMFunctionReturnNode(outputQuat=@n2.Result, outputWeight=@Multiply.Result)
  ]}
 New Function_ContainedGraph {nodes [
   n1 = RigVMFunctionEntryNode()
   > For_Each = RigVMDispatchNode(Array=@n1.transforms) {template="DISPATCH_RigVMDispatch_ArrayIterator(in Array,out Element,out Index,out Count,out Ratio)"}
   > &RigUnit_OffsetTransformForItem(Item=@For_Each.Element.Item, OffsetTransform=@For_Each.Element.Transform, Weight=@n1.weights, bPropagateToChildren=@n1.inheritsTransform)
   RigVMFunctionReturnNode(ExecuteContext=@For_Each.Completed)
  ]}
 New Function_ContainedGraph {nodes [
   n1 = RigVMFunctionEntryNode()
   RigVMFunction_Sequence = &RigVMFunction_Sequence(ExecuteContext=@n1.ExecuteContext)
   n2 = &RigUnit_SphericalPoseReader(
    ActiveRegionScaleFactors=(PositiveWidth=1,NegativeWidth=1,PositiveHeight=2.8,NegativeHeight=2)
    ActiveRegionSize=0.08
    Debug=(bDrawDebug=false,bDraw2D=False,bDrawLocalAxes=False,DebugScale=25,DebugSegments=20,DebugThickness=0.25)
    DriverAxis=(X=1,Y=0,Z=0)
    DriverItem=(Type=Bone,Name="head")
    FalloffRegionScaleFactors=(PositiveWidth=1,NegativeWidth=1,PositiveHeight=6,NegativeHeight=1)
    FalloffSize=0.25
    FlipHeightScaling=False
    FlipWidthScaling=False
    OptionalParentItem=(Type=Bone,Name="neck_01")
    RotationOffset=(X=70,Y=0,Z=-135)
    ExecutePin=@RigVMFunction_Sequence.A
   )
   > &RigUnit_ModifyTransforms(
    ItemToModify=((Item=(Type=Bone,Name="hat"),Transform=(Rotation=(X=0,Y=0,Z=0.0349,W=0.9994),Translation=(X=0,Y=0,Z=0),Scale3D=(X=0.9,Y=1,Z=1))),(Item=(Type=Bone,Name="head_back"),Transform=(Rotation=(X=0,Y=0,Z=0,W=1),Translation=(X=0,Y=-2,Z=0),Scale3D=(X=1,Y=1,Z=1))))
    WeightMinimum=0
    Weight=@n2.OutputParam
   )
   > DeltaRotation v4_7 = RigVMFunctionReferenceNode(
    Rotation Order=ZYX
    child=(Type=Bone,Name="head")
    debugRotation=false
    outputMinMax=(X=0,Y=1)
    parent=(Type=Null,Name="head_null")
    premult=1
    rotationMinMax=(X=-15,Y=-70)
    triggerAxis=Z
    twistAxis=(X=1,Y=0,Z=0)
   )
   n3 = &RigVMFunction_AlphaInterp(
    InterpSpeedDecreasing=15
    InterpSpeedIncreasing=15
    Scale=1
    bClampResult=False
    Value=&RigVMFunction_MathBoolToFloat(Value=RigVMVariableNode() {variable=isSkydiving_CR}.Value).Result
   )
   Multiply = &RigVMFunction_MathDoubleMul(B=@n3.Result, A=@"DeltaRotation v4_7".outputWeight)
   > &RigUnit_ModifyTransforms(
    ItemToModify=((Item=(Type=Bone,Name="hat"),Transform=(Rotation=(X=0,Y=0,Z=0.0698,W=0.9976),Translation=(X=0,Y=0,Z=0),Scale3D=(X=1,Y=0.9,Z=1))))
    WeightMinimum=0
    Weight=@Multiply.Result
   )
   > DeltaRotation v4_4 = RigVMFunctionReferenceNode(
    Rotation Order=ZYX
    child=(Type=Bone,Name="head")
    debugRotation=false
    outputMinMax=(X=0,Y=1)
    parent=(Type=Null,Name="head_null")
    premult=1
    rotationMinMax=(X=-45,Y=-55)
    triggerAxis=Z
    twistAxis=(X=1,Y=0,Z=0)
   )
   n4 = &RigVMFunction_AlphaInterp(
    InterpSpeedDecreasing=15
    InterpSpeedIncreasing=15
    Scale=1
    bClampResult=False
    Value=&RigVMFunction_MathBoolToFloat(Value=RigVMVariableNode() {variable=isCrouching_CR}.Value).Result
   )
   n5 = &RigVMFunction_MathDoubleMul(B=@n4.Result, A=@"DeltaRotation v4_4".outputWeight)
   > &RigUnit_ModifyTransforms(
    ItemToModify=((Item=(Type=Bone,Name="hat"),Transform=(Rotation=(X=0,Y=0,Z=0.1305,W=0.9914),Translation=(X=0,Y=0,Z=0),Scale3D=(X=1,Y=1,Z=1))))
    WeightMinimum=0
    Weight=@n5.Result
   )
   > AlterTransform_0 = RigVMFunctionReferenceNode(
    inheritsTransform=false
    transforms=((Item=(Type=Bone,Name="dyn_hair_1_c"),Transform=(Rotation=(X=0,Y=0,Z=0,W=1),Translation=(X=0,Y=0,Z=5),Scale3D=(X=1,Y=1,Z=1))),(Item=(Type=Bone,Name="dyn_hair_1_l"),Transform=(Rotation=(X=0,Y=0,Z=0,W=1),Translation=(X=0,Y=0,Z=5),Scale3D=(X=1,Y=1,Z=1))))
    weights=@n5.Result
   )
   RigVMFunctionReturnNode(ExecuteContext=@RigVMFunction_Sequence.B)
  ]}
 RigVMFunctionLibrary {nodes [
   AlterTransform = RigVMCollapseNode(ExecuteContext=(), inheritsTransform=false, transforms=(), weights=1)
   DeltaRotation v4 = RigVMCollapseNode(
    ExecuteContext=()
    Rotation Order=ZYX
    child=(Type=Bone,Name="None")
    outputMinMax=(X=0,Y=1)
    parent=(Type=Bone,Name="None")
    premult=1
    rotationMinMax=(X=0,Y=0)
    twistAxis=(X=1,Y=0,Z=0)
   )
   constraints = RigVMCollapseNode(ExecuteContext=())
   hair = RigVMCollapseNode(ExecuteContext=())
   hat = RigVMCollapseNode(ExecuteContext=())
  ]}
 RigVMFunction_Sequence_1_ContainedGraph {nodes [
   n1 = RigVMFunctionEntryNode()
   RigVMFunction_Sequence = &RigVMFunction_Sequence(ExecuteContext=@n1.ExecuteContext)
   n2 = &RigVMFunction_Sequence(ExecuteContext=@RigVMFunction_Sequence.B)
   n3 = &RigVMFunction_Sequence(ExecuteContext=@n2.B)
   RigVMFunctionReturnNode(A=@RigVMFunction_Sequence.A, B=@n2.A, C=@n3.A, D=@n3.B)
  ]}
 RigVMModel {nodes [
   RigUnit_BeginExecution = &RigUnit_BeginExecution()
   constraints = RigVMFunctionReferenceNode(ExecuteContext=@RigUnit_BeginExecution.ExecutePin)
   > hair = RigVMFunctionReferenceNode()
   > hat = RigVMFunctionReferenceNode()
  ]}
]
hierarchy [
 root {loc=<0 0 0>, rot=<0 0 0 1>}
 pelvis {parent=root, loc=<0 -0.5301 98.0571>, rot=<0.0127 0.707 -0.0127 -0.707>}
 spine_01 {parent=pelvis, loc=<6.1208 -0.9328 0>, rot=<0 0 0.034 -0.9994>}
 spine_02 {parent=spine_01, loc=<6.5997 -0.2095 0>, rot=<0 0 0.0238 -0.9997>}
 spine_03 {parent=spine_02, loc=<6.4914 -0.1237 0>, rot=<0 0 0.0149 -0.9999>}
 spine_04 {parent=spine_03, loc=<6.9598 0.0794 0>, rot=<0 0 -0.0132 -0.9999>}
 spine_05 {parent=spine_04, loc=<6.9987 0 0>, rot=<0 0 -0.0015 -1>}
 neck_01 {parent=spine_05, loc=<11.3138 0.1529 0>, rot=<0 0 -0.2003 -0.9797>}
 neck_02 {parent=neck_01, loc=<4.5627 0 0>, rot=<0 0 0.0504 -0.9987>}
 head {parent=neck_02, loc=<5.3914 0 0>, rot=<0 0 0.115 -0.9934>}
 base_hair_1_r {parent=head, loc=<2.7104 -0.019 7.4567>, rot=<0.2332 -0.6675 0.6675 -0.2332>}
 dyn_hair_1_r {parent=base_hair_1_r, loc=<0 0 0>, rot=<0 0 0 1>}
 dyn_hair_2_r {parent=dyn_hair_1_r, loc=<8.2 0 -0.0001>, rot=<0 0.0399 0 -0.9992>}
 dyn_hair_3_r {parent=dyn_hair_2_r, loc=<8.2 0 0.0002>, rot=<0 -0.1688 0 -0.9857>}
 dyn_hair_4_r {parent=dyn_hair_3_r, loc=<8.2 0 -0.0001>, rot=<0 -0.1728 0 -0.985>}
 dyn_hair_5_r {parent=dyn_hair_4_r, loc=<8.2 0 0.0001>, rot=<0 -0.0514 0 -0.9987>}
 dyn_hair_push_r {parent=dyn_hair_2_r, loc=<1.6442 -2.7352 -2.9846>, rot=<0 -0.3168 0 -0.9485>}
 base_hair_1_l {parent=head, loc=<2.7078 -0.1008 -7.9234>, rot=<0.2281 -0.6695 0.6682 -0.2308>}
 dyn_hair_1_l {parent=base_hair_1_l, loc=<0 0 0>, rot=<0 0 0 1>}
 dyn_hair_2_l {parent=dyn_hair_1_l, loc=<8.2 0 0>, rot=<0 0.0399 0 -0.9992>}
 dyn_hair_3_l {parent=dyn_hair_2_l, loc=<8.2 0 0>, rot=<0 -0.1485 0 -0.9889>}
 dyn_hair_4_l {parent=dyn_hair_3_l, loc=<8.2 0 0>, rot=<0 -0.1985 0 -0.9801>}
 dyn_hair_5_l {parent=dyn_hair_4_l, loc=<8.2 0 0>, rot=<0 -0.0514 0 -0.9987>}
 dyn_hair_push_l {parent=dyn_hair_2_l, loc=<1.4727 2.814 -4.2002>, rot=<0 -0.3168 0 -0.9485>}
 base_hair_1_c {parent=head, loc=<1.8493 -6.6562 0>, rot=<0.1334 -0.6944 0.6944 -0.1334>}
 dyn_hair_1_c {parent=base_hair_1_c, loc=<0 0 0>, rot=<0 0 0 1>}
 dyn_hair_2_c {parent=dyn_hair_1_c, loc=<7.3205 0 0>, rot=<0 0.0399 0 -0.9992>}
 dyn_hair_3_c {parent=dyn_hair_2_c, loc=<7.3197 0 0>, rot=<0 -0.0075 0 -1>}
 dyn_hair_4_c {parent=dyn_hair_3_c, loc=<7.2884 0 0>, rot=<0 -0.1985 0 -0.9801>}
 dyn_hair_5_c {parent=dyn_hair_4_c, loc=<7.2939 0 0>, rot=<0 -0.0514 0 -0.9987>}
 hat {parent=head, loc=<8.9142 9.1767 0>, rot=<0 0 0 1>}
 base_hatTassels_01_l {parent=hat, loc=<3.7199 -7.4662 -13.5126>, rot=<0 0.6494 0.7604 0>}
 dyn_hatTassels_01_l {parent=base_hatTassels_01_l, loc=<0 0 0>, rot=<0 0 0 1>}
 base_hatTassels_02_l {parent=hat, loc=<2.9202 -11.5239 -13.8426>, rot=<0 0.7071 0.7071 0>}
 dyn_hatTassels_02_l {parent=base_hatTassels_02_l, loc=<0 0 0>, rot=<0 0 0 1>}
 base_hatTassels_03_l {parent=hat, loc=<1.4602 -15.0975 -13.5727>, rot=<0 -0.7621 -0.6475 0>}
 dyn_hatTassels_03_l {parent=base_hatTassels_03_l, loc=<0 0 0>, rot=<0 0 0 1>}
 base_hatTassels_01_r {parent=hat, loc=<3.7199 -7.4662 13.573>, rot=<0 -0.6494 0.7604 0>}
 dyn_hatTassels_01_r {parent=base_hatTassels_01_r, loc=<0 0 0>, rot=<0 0 0 1>}
 base_hatTassels_02_r {parent=hat, loc=<2.9202 -11.5239 13.573>, rot=<0 -0.7071 0.7071 0>}
 dyn_hatTassels_02_r {parent=base_hatTassels_02_r, loc=<0 0 0>, rot=<0 0 0 1>}
 base_hatTassels_03_r {parent=hat, loc=<1.4602 -15.0975 13.573>, rot=<0 -0.7621 0.6475 0>}
 dyn_hatTassels_03_r {parent=base_hatTassels_03_r, loc=<0 0 0>, rot=<0 0 0 1>}
 head_back {parent=head, loc=<6.9362 -6.7204 0>, rot=<0 0 0 1>}
 head_null {parent=neck_02, type=Null, loc=<5.3914 0 0>, rot=<0 0 0.115 -0.9934>}
 clavicle_l {parent=spine_05, loc=<5.4831 -0.1053 -3.2946>, rot=<0.0465 -0.7289 -0.009 -0.6829>}
 upperarm_l {parent=clavicle_l, loc=<9.8658 -1.2136 -0.4105>, rot=<0.0164 -0.3245 -0.0269 -0.9454>}
 upperarm_l_Support {parent=clavicle_l, loc=<9.8658 -1.2136 -0.4105>, rot=<0.0164 -0.3245 -0.0269 -0.9454>}
 clavicle_r {parent=spine_05, loc=<5.4831 -0.1053 3.2946>, rot=<0.7289 0.0465 -0.6829 0.009>}
 upperarm_r {parent=clavicle_r, loc=<-9.8658 1.2136 0.4105>, rot=<0.0164 -0.3245 -0.0269 -0.9454>}
 dyn_hair_1_c_ctrl_null {type=Null, loc=<0 -8.3586 153.8284>, rot=<0.3991 0.3991 -0.5837 0.5837>}
 dyn_hair_1_c_ctrl {controlType=EULER_TRANSFORM, parent=dyn_hair_1_c_ctrl_null, type=Control, loc=<0 0 0>, rot=<0 0 0 1>}
 dyn_hair_2_c_ctrl_null {type=Null, loc=<0 -11.0142 147.0067>, rot=<-0.3755 -0.3755 0.5992 -0.5992>}
 dyn_hair_2_c_ctrl {controlType=EULER_TRANSFORM, parent=dyn_hair_2_c_ctrl_null, type=Control, loc=<-7.2971 0 0.5838>, rot=<0 -0.0399 0 -0.9992>}
 dyn_hair_3_c_ctrl_null {type=Null, loc=<0 -14.2051 140.4191>, rot=<-0.38 -0.38 0.5963 -0.5963>}
 dyn_hair_3_c_ctrl {controlType=EULER_TRANSFORM, parent=dyn_hair_3_c_ctrl_null, type=Control, loc=<-7.3189 0 -0.1098>, rot=<0 0.0075 0 -1>}
 dyn_hair_4_c_ctrl_null {type=Null, loc=<0 -17.2836 133.8127>, rot=<-0.4908 -0.4908 0.509 -0.509>}
 dyn_hair_4_c_ctrl {controlType=EULER_TRANSFORM, parent=dyn_hair_4_c_ctrl_null, type=Control, loc=<-6.7141 0 -2.8359>, rot=<0 0.1985 0 -0.9801>}
 dyn_hair_5_c_ctrl_null {type=Null, loc=<0 -17.5492 126.5236>, rot=<-0.5163 -0.5163 0.4831 -0.4831>}
 dyn_hair_5_c_ctrl {controlType=EULER_TRANSFORM, parent=dyn_hair_5_c_ctrl_null, type=Control, loc=<-7.2554 0 -0.749>, rot=<0 0.0514 0 -0.9987>}
 dyn_hair_1_r_ctrl_null {type=Null, loc=<-7.4567 -1.7287 154.7444>, rot=<0.3097 0.3097 -0.6357 0.6357>}
 dyn_hair_1_r_ctrl {
  controlType=EULER_TRANSFORM
  parent=dyn_hair_1_r_ctrl_null
  type=Control
  loc=<120.7966 7.4567 -96.7316>
  rot=<-0.3097 -0.3097 0.6357 0.6357>
 }
 dyn_hair_2_r_ctrl_null {type=Null, loc=<-7.4567 -6.7824 148.2868>, rot=<-0.2841 -0.2841 0.6475 -0.6475>}
 dyn_hair_2_r_ctrl {controlType=EULER_TRANSFORM, parent=dyn_hair_2_r_ctrl_null, type=Control, loc=<-8.1739 0 0.6541>, rot=<0 -0.0399 0 -0.9992>}
 dyn_hair_3_r_ctrl_null {type=Null, loc=<-7.4567 -12.3352 142.2531>, rot=<-0.3893 -0.3893 0.5903 -0.5903>}
 dyn_hair_3_r_ctrl {controlType=EULER_TRANSFORM, parent=dyn_hair_3_r_ctrl_null, type=Control, loc=<-7.7327 0 -2.7285>, rot=<0 0.1688 0 -0.9857>}
 dyn_hair_4_r_ctrl_null {type=Null, loc=<-7.4567 -15.5638 134.7154>, rot=<-0.4855 -0.4855 0.5141 -0.5141>}
 dyn_hair_4_r_ctrl {controlType=EULER_TRANSFORM, parent=dyn_hair_4_r_ctrl_null, type=Control, loc=<-7.7104 0 -2.7911>, rot=<0 0.1728 0 -0.985>}
 dyn_hair_5_r_ctrl_null {type=Null, loc=<-7.4567 -16.034 126.5289>, rot=<-0.5112 -0.5112 0.4885 -0.4885>}
 dyn_hair_5_r_ctrl {controlType=EULER_TRANSFORM, parent=dyn_hair_5_r_ctrl_null, type=Control, loc=<-8.1566 0 -0.8421>, rot=<0 0.0514 0 -0.9987>}
 dyn_hair_1_l_ctrl_null {type=Null, loc=<7.9234 -1.8106 154.7411>, rot=<0.3138 0.3129 -0.6325 0.6353>}
 dyn_hair_1_l_ctrl {
  controlType=EULER_TRANSFORM
  parent=dyn_hair_1_l_ctrl_null
  type=Control
  loc=<121.8097 -8.3813 -95.4097>
  rot=<-0.3138 -0.3129 0.6325 0.6353>
 }
 dyn_hair_2_l_ctrl_null {type=Null, loc=<7.9574 -6.79 148.2262>, rot=<-0.2883 -0.2873 0.6445 -0.6473>}
 dyn_hair_2_l_ctrl {controlType=EULER_TRANSFORM, parent=dyn_hair_2_l_ctrl_null, type=Control, loc=<-8.1739 0 0.654>, rot=<0 -0.0399 0 -0.9992>}
 dyn_hair_3_l_ctrl_null {type=Null, loc=<7.9917 -12.2732 142.1292>, rot=<-0.3809 -0.3802 0.5945 -0.5974>}
 dyn_hair_3_l_ctrl {controlType=EULER_TRANSFORM, parent=dyn_hair_3_l_ctrl_null, type=Control, loc=<-7.8383 0 -2.4086>, rot=<0 0.1485 0 -0.9889>}
 dyn_hair_4_l_ctrl_null {type=Null, loc=<8.024 -15.7235 134.6905>, rot=<-0.4913 -0.4912 0.5071 -0.5101>}
 dyn_hair_4_l_ctrl {controlType=EULER_TRANSFORM, parent=dyn_hair_4_l_ctrl_null, type=Control, loc=<-7.5538 0 -3.1906>, rot=<0 0.1985 0 -0.9801>}
 dyn_hair_5_l_ctrl_null {type=Null, loc=<8.0494 -16.0077 126.4955>, rot=<-0.5167 -0.5168 0.4812 -0.4841>}
 dyn_hair_5_l_ctrl {controlType=EULER_TRANSFORM, parent=dyn_hair_5_l_ctrl_null, type=Control, loc=<-8.1567 0 -0.842>, rot=<0 0.0514 0 -0.9987>}
]
variables [
 HeadRotLeft {type=double, default=0}
 isCrouching_CR {type=bool, default=False}
 isSkydiving_CR {type=bool, default=False}
]
