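# Post-process style AnimBP for the F_MED_ForestJeanWinkHat face accessory, adding secondary motion to hair and hat tassels.
# The copied parent pose runs through a Control Rig fed skydive/crouch state, then AnimDynamics simulates hair, bypassed for backblings or disabled rigid bodies.
# Tassels always simulate, followed by additive hair_4 bone offsets; the hair constraint chain is unconnected.
# Each update refreshes the bypass flag and computes GravityOverride from the ForestJeanWink parameters asset.
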AnimBlueprint /BRCosmetics/Characters/Player/Female/Medium/Bodies/F_MED_ForestJeanWinkHat/Meshes/Parts/F_MED_ForestJeanWinkHat_FaceAcc_AnimBP
target_skeleton=&/F_MED_ForestJeanWinkHat_FaceAcc_Skeleton
graphs [
 AnimGraph {nodes [
   n1 = "Constraint - dyn_hair_3_r (PPPP)"!(
    ComponentPose
    Alpha=1
    ConstraintWeights_0=0.7335
    ConstraintWeights_1:float=0.2408
    ConstraintWeights_2:float=0.023
    ConstraintWeights_3:float=0.0027
   ) -> (Pose) {
    boneToModify {boneName=dyn_hair_3_r}
    constraintSetup [{targetBone {boneName=clavicle_r}}, {targetBone {boneName=spine_05}}, {targetBone {boneName=upperarm_r}}, {targetBone {boneName=spine_04}}]
    constraintWeights [1,1,1,1]
   }
   n2 = "Constraint - dyn_hair_4_r (PPPP)"!(
    ComponentPose=@n1
    Alpha=1
    ConstraintWeights_0=0.6475
    ConstraintWeights_1:float=0.2451
    ConstraintWeights_2:float=0.059
    ConstraintWeights_3:float=0.0484
   ) -> (Pose) {
    boneToModify {boneName=dyn_hair_4_r}
    constraintSetup [{targetBone {boneName=clavicle_r}}, {targetBone {boneName=spine_05}}, {targetBone {boneName=spine_04}}, {targetBone {boneName=upperarm_r}}]
    constraintWeights [1,1,1,1]
   }
   n3 = "Constraint - dyn_hair_5_r (PPPPPPP)"!(
    ComponentPose=@n2
    Alpha=1
    ConstraintWeights_0=0.3312
    ConstraintWeights_1:float=0.2805
    ConstraintWeights_2:float=0.2268
    ConstraintWeights_3:float=0.0761
    ConstraintWeights_4:float=0.0376
    ConstraintWeights_5:float=0.0372
    ConstraintWeights_6:float=0.0107
   ) -> (Pose) {
    boneToModify {boneName=dyn_hair_5_r}
    constraintSetup [
     {targetBone {boneName=clavicle_r}}
     {targetBone {boneName=spine_04}}
     {targetBone {boneName=spine_05}}
     {targetBone {boneName=upperarm_r}}
     {targetBone {boneName=spine_03}}
     {targetBone {boneName=spine_02}}
     {targetBone {boneName=clavicle_l}}
    ]
    constraintWeights [1,1,1,1,1,1,1]
   }
   n4 = "Constraint - dyn_hair_3_l (PPPP)"!(
    ComponentPose=@n3
    Alpha=1
    ConstraintWeights_0=0.7335
    ConstraintWeights_1:float=0.2408
    ConstraintWeights_2:float=0.023
    ConstraintWeights_3:float=0.0027
   ) -> (Pose) {
    boneToModify {boneName=dyn_hair_3_l}
    constraintSetup [{targetBone {boneName=clavicle_l}}, {targetBone {boneName=spine_05}}, {targetBone {boneName=upperarm_l}}, {targetBone {boneName=spine_04}}]
    constraintWeights [1,1,1,1]
   }
   n5 = "Constraint - dyn_hair_4_l (PPPP)"!(
    ComponentPose=@n4
    Alpha=1
    ConstraintWeights_0=0.6475
    ConstraintWeights_1:float=0.2451
    ConstraintWeights_2:float=0.059
    ConstraintWeights_3:float=0.0484
   ) -> (Pose) {
    boneToModify {boneName=dyn_hair_4_l}
    constraintSetup [{targetBone {boneName=clavicle_l}}, {targetBone {boneName=spine_05}}, {targetBone {boneName=spine_04}}, {targetBone {boneName=upperarm_l}}]
    constraintWeights [1,1,1,1]
   }
   n6 = "Constraint - dyn_hair_5_l (PPPPPPP)"!(
    ComponentPose=@n5
    Alpha=1
    ConstraintWeights_0=0.3312
    ConstraintWeights_1:float=0.2805
    ConstraintWeights_2:float=0.2268
    ConstraintWeights_3:float=0.0761
    ConstraintWeights_4:float=0.0376
    ConstraintWeights_5:float=0.0372
    ConstraintWeights_6:float=0.0107
   ) -> (Pose) {
    boneToModify {boneName=dyn_hair_5_l}
    constraintSetup [
     {targetBone {boneName=clavicle_l}}
     {targetBone {boneName=spine_04}}
     {targetBone {boneName=spine_05}}
     {targetBone {boneName=upperarm_l}}
     {targetBone {boneName=spine_03}}
     {targetBone {boneName=spine_02}}
     {targetBone {boneName=clavicle_r}}
    ]
    constraintWeights [1,1,1,1,1,1,1]
   }
   n7 = "Constraint - dyn_hair_3_c (PPP)"!(ComponentPose=@n6, Alpha=1, ConstraintWeights_0=0.8616, ConstraintWeights_1:float=0.0703, ConstraintWeights_2:float=0.068) -> (Pose) {
    boneToModify {boneName=dyn_hair_3_c}
    constraintSetup [{targetBone {boneName=spine_05}}, {targetBone {boneName=clavicle_l}}, {targetBone {boneName=clavicle_r}}]
    constraintWeights [1,1,1]
   }
   n8 = "Constraint - dyn_hair_4_c (PPPP)"!(
    ComponentPose=@n7
    Alpha=1
    ConstraintWeights_0=0.6005
    ConstraintWeights_1:float=0.279
    ConstraintWeights_2:float=0.0891
    ConstraintWeights_3:float=0.0314
   ) -> (Pose) {
    boneToModify {boneName=dyn_hair_4_c}
    constraintSetup [{targetBone {boneName=spine_05}}, {targetBone {boneName=clavicle_r}}, {targetBone {boneName=clavicle_l}}, {targetBone {boneName=spine_04}}]
    constraintWeights [1,1,1,1]
   }
   "Constraint - dyn_hair_5_c (PPPPPPP)"!(
    ComponentPose=@n8
    Alpha=1
    ConstraintWeights_0=0.3131
    ConstraintWeights_1:float=0.3033
    ConstraintWeights_2:float=0.1729
    ConstraintWeights_3:float=0.0671
    ConstraintWeights_4:float=0.0629
    ConstraintWeights_5:float=0.0616
    ConstraintWeights_6:float=0.019
   ) -> (Pose) {
    boneToModify {boneName=dyn_hair_5_c}
    constraintSetup [
     {targetBone {boneName=spine_04}}
     {targetBone {boneName=spine_05}}
     {targetBone {boneName=clavicle_r}}
     {targetBone {boneName=clavicle_l}}
     {targetBone {boneName=spine_02}}
     {targetBone {boneName=spine_03}}
     {targetBone {boneName=spine_01}}
    ]
    constraintWeights [1,1,1,1,1,1,1]
   }
   n9 = Anim Dynamics - Bone: dyn_hair_2_l!(ComponentPose=Local To Component(LocalPose=Use cached pose 'preSimCache'()), Alpha=1, GravityOverride:Vector="0,0,-980") -> (Pose) {
    angularDampingOverride=0.9999
    angularSpringConstant=3000
    bAngularSpring=true
    bOverrideAngularDamping=true
    bUseGravityOverride=true
    gravityScale=0.35
    simulationSpace=RootRelative
    boundBone {boneName=dyn_hair_2_l}
    componentAppliedLinearAccClamp=<3000 3000 3000>
    componentLinearAccScale=<0.3 0.3 0.75>
    componentLinearVelScale=<0.3 0.3 0.75>
    gravityOverride {z=-980}
    physicsBodyDefinitions [{
      boundBone {boneName=dyn_hair_2_l}
      boxExtents=<25 25 25>
      constraintSetup {angularLimitsMax=<0 0 5>, angularLimitsMin=<0 -30 -5>, angularTarget=<1 0 0>}
      localJointOffset=<1 0 0>
     }]
    simSpaceSettings {simSpaceAngularAlpha=0.1, externalAngularVelocity {z=0.5}}
   }
   n10 = Anim Dynamics - Bone: dyn_hair_2_c!(ComponentPose=@n9, Alpha=1, GravityOverride:Vector="0,0,-980") -> (Pose) {
    angularDampingOverride=0.9999
    angularSpringConstant=3000
    bAngularSpring=true
    bOverrideAngularDamping=true
    bUseGravityOverride=true
    gravityScale=0.35
    simulationSpace=RootRelative
    boundBone {boneName=dyn_hair_2_c}
    componentAppliedLinearAccClamp=<3000 3000 3000>
    componentLinearAccScale=<0.3 0.3 0.75>
    componentLinearVelScale=<0.3 0.3 0.75>
    gravityOverride {z=-980}
    physicsBodyDefinitions [{
      boundBone {boneName=dyn_hair_2_c}
      boxExtents=<25 25 25>
      constraintSetup {angularLimitsMax=<0 0 5>, angularLimitsMin=<0 -30 -5>, angularTarget=<1 0 0>}
      localJointOffset=<1 0 0>
     }]
    simSpaceSettings {simSpaceAngularAlpha=0.1, externalAngularVelocity {z=0.5}}
   }
   n11 = Anim Dynamics - Bone: dyn_hair_2_r!(ComponentPose=@n10, Alpha=1, GravityOverride:Vector="0,0,-980") -> (Pose) {
    angularDampingOverride=0.9999
    angularSpringConstant=3000
    bAngularSpring=true
    bOverrideAngularDamping=true
    bUseGravityOverride=true
    gravityScale=0.35
    simulationSpace=RootRelative
    boundBone {boneName=dyn_hair_2_r}
    componentAppliedLinearAccClamp=<3000 3000 3000>
    componentLinearAccScale=<0.3 0.3 0.75>
    componentLinearVelScale=<0.3 0.3 0.75>
    gravityOverride {z=-980}
    physicsBodyDefinitions [{
      boundBone {boneName=dyn_hair_2_r}
      boxExtents=<25 25 25>
      constraintSetup {angularLimitsMax=<0 0 5>, angularLimitsMin=<0 -30 -5>, angularTarget=<1 0 0>}
      localJointOffset=<1 0 0>
     }]
    simSpaceSettings {simSpaceAngularAlpha=0.1, externalAngularVelocity {z=0.5}}
   }
   n12 = Anim Dynamics - Bone: dyn_hair_3_l!(ComponentPose=@n11, Alpha=1, GravityOverride:Vector="0,0,-980") -> (Pose) {
    angularDampingOverride=0.9999
    angularSpringConstant=3000
    bAngularSpring=true
    bOverrideAngularDamping=true
    bUseGravityOverride=true
    gravityScale=0.35
    simulationSpace=RootRelative
    boundBone {boneName=dyn_hair_3_l}
    componentAppliedLinearAccClamp=<3000 3000 3000>
    componentLinearAccScale=<0.3 0.3 0.65>
    componentLinearVelScale=<0.3 0.3 0.65>
    gravityOverride {z=-980}
    physicsBodyDefinitions [{
      boundBone {boneName=dyn_hair_3_l}
      boxExtents=<25 25 25>
      constraintSetup {angularLimitsMax=<0 0 10>, angularLimitsMin=<0 -35 -10>, angularTarget=<1 0 0>}
      localJointOffset=<5 0 0>
     }]
    simSpaceSettings {simSpaceAngularAlpha=0.1, externalAngularVelocity {z=0.5}}
   }
   n13 = Anim Dynamics - Bone: dyn_hair_3_c!(ComponentPose=@n12, Alpha=1, GravityOverride:Vector="0,0,-980") -> (Pose) {
    angularDampingOverride=0.9999
    angularSpringConstant=3000
    bAngularSpring=true
    bOverrideAngularDamping=true
    bUseGravityOverride=true
    gravityScale=0.35
    simulationSpace=RootRelative
    boundBone {boneName=dyn_hair_3_c}
    componentAppliedLinearAccClamp=<3000 3000 3000>
    componentLinearAccScale=<0.3 0.3 0.65>
    componentLinearVelScale=<0.3 0.3 0.65>
    gravityOverride {z=-980}
    physicsBodyDefinitions [{
      boundBone {boneName=dyn_hair_3_c}
      boxExtents=<25 25 25>
      constraintSetup {angularLimitsMax=<0 0 10>, angularLimitsMin=<0 -35 -10>, angularTarget=<1 0 0>}
      localJointOffset=<5 0 0>
     }]
    simSpaceSettings {simSpaceAngularAlpha=0.1, externalAngularVelocity {z=0.5}}
   }
   n14 = Anim Dynamics - Bone: dyn_hair_3_r!(ComponentPose=@n13, Alpha=1, GravityOverride:Vector="0,0,-980") -> (Pose) {
    angularDampingOverride=0.9999
    angularSpringConstant=3000
    bAngularSpring=true
    bOverrideAngularDamping=true
    bUseGravityOverride=true
    gravityScale=0.35
    simulationSpace=RootRelative
    boundBone {boneName=dyn_hair_3_r}
    componentAppliedLinearAccClamp=<3000 3000 3000>
    componentLinearAccScale=<0.3 0.3 0.65>
    componentLinearVelScale=<0.3 0.3 0.65>
    gravityOverride {z=-980}
    physicsBodyDefinitions [{
      boundBone {boneName=dyn_hair_3_r}
      boxExtents=<25 25 25>
      constraintSetup {angularLimitsMax=<0 0 10>, angularLimitsMin=<0 -35 -10>, angularTarget=<1 0 0>}
      localJointOffset=<5 0 0>
     }]
    simSpaceSettings {simSpaceAngularAlpha=0.1, externalAngularVelocity {z=0.5}}
   }
   n15 = Anim Dynamics - Bone: dyn_hair_4_l!(ComponentPose=@n14, Alpha=1, GravityOverride:Vector="0,0,-980") -> (Pose) {
    angularDampingOverride=0.9999
    angularSpringConstant=3000
    bAngularSpring=true
    bOverrideAngularDamping=true
    bUseGravityOverride=true
    gravityScale=0.35
    simulationSpace=RootRelative
    boundBone {boneName=dyn_hair_4_l}
    componentAppliedLinearAccClamp=<3000 3000 3000>
    componentLinearAccScale=<0.3 0.3 0.65>
    componentLinearVelScale=<0.3 0.3 0.65>
    gravityOverride {z=-980}
    physicsBodyDefinitions [{
      boundBone {boneName=dyn_hair_4_l}
      boxExtents=<25 25 25>
      constraintSetup {angularLimitsMax=<0 0 10>, angularLimitsMin=<0 -30 -10>, angularTarget=<1 0 0>}
      localJointOffset=<5 0 0>
     }]
    simSpaceSettings {simSpaceAngularAlpha=0.1, externalAngularVelocity {z=0.5}}
   }
   n16 = Anim Dynamics - Bone: dyn_hair_4_c!(ComponentPose=@n15, Alpha=1, GravityOverride:Vector="0,0,-980") -> (Pose) {
    angularDampingOverride=0.9999
    angularSpringConstant=3000
    bAngularSpring=true
    bOverrideAngularDamping=true
    bUseGravityOverride=true
    gravityScale=0.35
    simulationSpace=RootRelative
    boundBone {boneName=dyn_hair_4_c}
    componentAppliedLinearAccClamp=<3000 3000 3000>
    componentLinearAccScale=<0.3 0.3 0.65>
    componentLinearVelScale=<0.3 0.3 0.65>
    gravityOverride {z=-980}
    physicsBodyDefinitions [{
      boundBone {boneName=dyn_hair_4_c}
      boxExtents=<25 25 25>
      constraintSetup {angularLimitsMax=<0 0 10>, angularLimitsMin=<0 -30 -10>, angularTarget=<1 0 0>}
      localJointOffset=<5 0 0>
     }]
    simSpaceSettings {simSpaceAngularAlpha=0.1, externalAngularVelocity {z=0.5}}
   }
   n17 = Anim Dynamics - Bone: dyn_hair_4_r!(ComponentPose=@n16, Alpha=1, GravityOverride:Vector="0,0,-980") -> (Pose) {
    angularDampingOverride=0.9999
    angularSpringConstant=3000
    bAngularSpring=true
    bOverrideAngularDamping=true
    bUseGravityOverride=true
    gravityScale=0.35
    simulationSpace=RootRelative
    boundBone {boneName=dyn_hair_4_r}
    componentAppliedLinearAccClamp=<3000 3000 3000>
    componentLinearAccScale=<0.3 0.3 0.65>
    componentLinearVelScale=<0.3 0.3 0.65>
    gravityOverride {z=-980}
    physicsBodyDefinitions [{
      boundBone {boneName=dyn_hair_4_r}
      boxExtents=<25 25 25>
      constraintSetup {angularLimitsMax=<0 0 10>, angularLimitsMin=<0 -30 -10>, angularTarget=<1 0 0>}
      localJointOffset=<5 0 0>
     }]
    simSpaceSettings {simSpaceAngularAlpha=0.1, externalAngularVelocity {z=0.5}}
   }
   n18 = Anim Dynamics - Bone: dyn_hair_5_l!(ComponentPose=@n17, Alpha=1, GravityOverride:Vector="0,0,-980") -> (Pose) {
    angularDampingOverride=0.9999
    angularSpringConstant=3000
    bAngularSpring=true
    bOverrideAngularDamping=true
    bUseGravityOverride=true
    gravityScale=0.35
    simulationSpace=RootRelative
    boundBone {boneName=dyn_hair_5_l}
    componentAppliedLinearAccClamp=<3000 3000 3000>
    componentLinearAccScale=<0.3 0.3 0.65>
    componentLinearVelScale=<0.3 0.3 0.65>
    gravityOverride {z=-980}
    physicsBodyDefinitions [{
      boundBone {boneName=dyn_hair_5_l}
      boxExtents=<25 25 25>
      constraintSetup {angularLimitsMax=<0 0 10>, angularLimitsMin=<0 -25 -10>, angularTarget=<1 0 0>}
      localJointOffset=<5 0 0>
     }]
    simSpaceSettings {simSpaceAngularAlpha=0.1, externalAngularVelocity {z=0.5}}
   }
   n19 = Anim Dynamics - Bone: dyn_hair_5_c!(ComponentPose=@n18, Alpha=1, GravityOverride:Vector="0,0,-980") -> (Pose) {
    angularDampingOverride=0.9999
    angularSpringConstant=3000
    bAngularSpring=true
    bOverrideAngularDamping=true
    bUseGravityOverride=true
    gravityScale=0.35
    simulationSpace=RootRelative
    boundBone {boneName=dyn_hair_5_c}
    componentAppliedLinearAccClamp=<3000 3000 3000>
    componentLinearAccScale=<0.3 0.3 0.65>
    componentLinearVelScale=<0.3 0.3 0.65>
    gravityOverride {z=-980}
    physicsBodyDefinitions [{
      boundBone {boneName=dyn_hair_5_c}
      boxExtents=<25 25 25>
      constraintSetup {angularLimitsMax=<0 0 10>, angularLimitsMin=<0 -25 -10>, angularTarget=<1 0 0>}
      localJointOffset=<5 0 0>
     }]
    simSpaceSettings {simSpaceAngularAlpha=0.1, externalAngularVelocity {z=0.5}}
   }
   n20 = Anim Dynamics - Bone: dyn_hair_5_r!(ComponentPose=@n19, Alpha=1, GravityOverride:Vector="0,0,-980") -> (Pose) {
    angularDampingOverride=0.9999
    angularSpringConstant=3000
    bAngularSpring=true
    bOverrideAngularDamping=true
    bUseGravityOverride=true
    gravityScale=0.35
    simulationSpace=RootRelative
    boundBone {boneName=dyn_hair_5_r}
    componentAppliedLinearAccClamp=<3000 3000 3000>
    componentLinearAccScale=<0.3 0.3 0.65>
    componentLinearVelScale=<0.3 0.3 0.65>
    gravityOverride {z=-980}
    physicsBodyDefinitions [{
      boundBone {boneName=dyn_hair_5_r}
      boxExtents=<25 25 25>
      constraintSetup {angularLimitsMax=<0 0 10>, angularLimitsMin=<0 -25 -10>, angularTarget=<1 0 0>}
      localJointOffset=<5 0 0>
     }]
    simSpaceSettings {simSpaceAngularAlpha=0.1, externalAngularVelocity {z=0.5}}
   }
   n21 = Component To Local(ComponentPose=@n20)
   n22 = Get isWearingBackbling / Cape / NoRBAN() -> (isWearingBackbling / Cape / NoRBAN:bool)
   n23 = Component To Local(ComponentPose=Local To Component(LocalPose=Use cached pose 'preSimCache'()))
   n24 = Anim Dynamics - Bone: dyn_hatTassels_01_l!(
    ComponentPose=Local To Component(LocalPose=Blend Poses by bool(BlendPose_0=@n23, BlendPose_1=@n21, bActiveValue=@n22))
    Alpha=1
    GravityOverride:Vector="0,0,-980"
    ExternalForce="0,0,0"
   ) -> (Pose) {
    angularDampingOverride=0.999
    angularSpringConstant=1000
    bAngularSpring=true
    bOverrideAngularDamping=true
    bUseGravityOverride=true
    gravityScale=0.35
    boundBone {boneName=dyn_hatTassels_01_l}
    componentAppliedLinearAccClamp=<3000 3000 3000>
    componentLinearAccScale=<0.3 0.3 0.3>
    componentLinearVelScale=<0.3 0.3 0.3>
    gravityOverride {z=-980}
    physicsBodyDefinitions [{
      boundBone {boneName=dyn_hatTassels_01_l}
      boxExtents=<10 15 15>
      constraintSetup {angularLimitsMax=<0 20 5>, angularLimitsMin=<0 -20 -55>, angularTarget=<1 0 0>}
      localJointOffset=<4 1 0>
     }]
    simSpaceSettings {simSpaceAngularAlpha=0.1}
   }
   n25 = Anim Dynamics - Bone: dyn_hatTassels_02_l!(ComponentPose=@n24, Alpha=1, GravityOverride:Vector="0,0,-980", ExternalForce="0,0,0") -> (Pose) {
    angularDampingOverride=0.999
    angularSpringConstant=1000
    bAngularSpring=true
    bOverrideAngularDamping=true
    bUseGravityOverride=true
    gravityScale=0.35
    boundBone {boneName=dyn_hatTassels_02_l}
    componentAppliedLinearAccClamp=<3000 3000 3000>
    componentLinearAccScale=<0.3 0.3 0.3>
    componentLinearVelScale=<0.3 0.3 0.3>
    gravityOverride {z=-980}
    physicsBodyDefinitions [{
      boundBone {boneName=dyn_hatTassels_02_l}
      boxExtents=<10 15 15>
      constraintSetup {angularLimitsMax=<0 20 5>, angularLimitsMin=<0 -20 -55>, angularTarget=<1 0 0>}
      localJointOffset=<4 1 0>
     }]
    simSpaceSettings {simSpaceAngularAlpha=0.1}
   }
   n26 = Anim Dynamics - Bone: dyn_hatTassels_03_l!(ComponentPose=@n25, Alpha=1, GravityOverride:Vector="0,0,-980", ExternalForce="0,0,0") -> (Pose) {
    angularDampingOverride=0.999
    angularSpringConstant=1000
    bAngularSpring=true
    bOverrideAngularDamping=true
    bUseGravityOverride=true
    gravityScale=0.35
    boundBone {boneName=dyn_hatTassels_03_l}
    componentAppliedLinearAccClamp=<3000 3000 3000>
    componentLinearAccScale=<0.3 0.3 0.3>
    componentLinearVelScale=<0.3 0.3 0.3>
    gravityOverride {z=-980}
    physicsBodyDefinitions [{
      boundBone {boneName=dyn_hatTassels_03_l}
      boxExtents=<10 15 15>
      constraintSetup {angularLimitsMax=<0 20 5>, angularLimitsMin=<0 -20 -55>, angularTarget=<1 0 0>}
      localJointOffset=<4 1 0>
     }]
    simSpaceSettings {simSpaceAngularAlpha=0.1}
   }
   n27 = Anim Dynamics - Bone: dyn_hatTassels_01_r!(ComponentPose=@n26, Alpha=1, GravityOverride:Vector="0,0,-980", ExternalForce="0,0,0") -> (Pose) {
    angularDampingOverride=0.999
    angularSpringConstant=1000
    bAngularSpring=true
    bOverrideAngularDamping=true
    bUseGravityOverride=true
    gravityScale=0.35
    boundBone {boneName=dyn_hatTassels_01_r}
    componentAppliedLinearAccClamp=<3000 3000 3000>
    componentLinearAccScale=<0.3 0.3 0.3>
    componentLinearVelScale=<0.3 0.3 0.3>
    gravityOverride {z=-980}
    physicsBodyDefinitions [{
      boundBone {boneName=dyn_hatTassels_01_r}
      boxExtents=<10 15 15>
      constraintSetup {angularLimitsMax=<0 20 5>, angularLimitsMin=<0 -20 -55>, angularTarget=<1 0 0>}
      localJointOffset=<4 1 0>
     }]
    simSpaceSettings {simSpaceAngularAlpha=0.1}
   }
   n28 = Anim Dynamics - Bone: dyn_hatTassels_02_r!(ComponentPose=@n27, Alpha=1, GravityOverride:Vector="0,0,-980", ExternalForce="0,0,0") -> (Pose) {
    angularDampingOverride=0.999
    angularSpringConstant=1000
    bAngularSpring=true
    bOverrideAngularDamping=true
    bUseGravityOverride=true
    gravityScale=0.35
    boundBone {boneName=dyn_hatTassels_02_r}
    componentAppliedLinearAccClamp=<3000 3000 3000>
    componentLinearAccScale=<0.3 0.3 0.3>
    componentLinearVelScale=<0.3 0.3 0.3>
    gravityOverride {z=-980}
    physicsBodyDefinitions [{
      boundBone {boneName=dyn_hatTassels_02_r}
      boxExtents=<10 15 15>
      constraintSetup {angularLimitsMax=<0 20 5>, angularLimitsMin=<0 -20 -55>, angularTarget=<1 0 0>}
      localJointOffset=<4 1 0>
     }]
    simSpaceSettings {simSpaceAngularAlpha=0.1}
   }
   n29 = Anim Dynamics - Bone: dyn_hatTassels_03_r!(ComponentPose=@n28, Alpha=1, GravityOverride:Vector="0,0,-980", ExternalForce="0,0,0") -> (Pose) {
    angularDampingOverride=0.999
    angularSpringConstant=1000
    bAngularSpring=true
    bOverrideAngularDamping=true
    bUseGravityOverride=true
    gravityScale=0.35
    boundBone {boneName=dyn_hatTassels_03_r}
    componentAppliedLinearAccClamp=<3000 3000 3000>
    componentLinearAccScale=<0.3 0.3 0.3>
    componentLinearVelScale=<0.3 0.3 0.3>
    gravityOverride {z=-980}
    physicsBodyDefinitions [{
      boundBone {boneName=dyn_hatTassels_03_r}
      boxExtents=<10 15 15>
      constraintSetup {angularLimitsMax=<0 20 5>, angularLimitsMin=<0 -20 -55>, angularTarget=<1 0 0>}
      localJointOffset=<4 1 0>
     }]
    simSpaceSettings {simSpaceAngularAlpha=0.1}
   }
   n30 = "Transform (Modify) Bone - Bone: dyn_hair_4_l"!(ComponentPose=@n29, bAlphaBoolEnabled:bool=True, Translation="0,0,0", Rotation="-20.000,0.000,-0.000", Scale="1,1,0.5") -> (Pose) {
    alphaInputType=Bool
    rotationMode=BMM_Additive
    rotationSpace=BCS_BoneSpace
    scaleMode=BMM_Additive
    scaleSpace=BCS_BoneSpace
    translationMode=BMM_Additive
    translationSpace=BCS_BoneSpace
    alphaBoolBlend {blendInTime=0.2, blendOutTime=0.2}
    boneToModify {boneName=dyn_hair_4_l}
    rotation {pitch=-20}
    scale {z=0.5}
   }
   n31 = "Transform (Modify) Bone - Bone: dyn_hair_4_r"!(ComponentPose=@n30, bAlphaBoolEnabled:bool=True, Translation="0,0,0", Rotation="0,0,0", Scale="1,1,0.8") -> (Pose) {
    alphaInputType=Bool
    rotationMode=BMM_Additive
    rotationSpace=BCS_BoneSpace
    scaleMode=BMM_Additive
    scaleSpace=BCS_BoneSpace
    translationMode=BMM_Additive
    translationSpace=BCS_BoneSpace
    alphaBoolBlend {blendInTime=0.2, blendOutTime=0.2}
    boneToModify {boneName=dyn_hair_4_r}
    scale {z=0.8}
   }
   n32 = "Transform (Modify) Bone - Bone: dyn_hair_4_c"!(ComponentPose=@n31, bAlphaBoolEnabled:bool=True, Translation="0,0,0", Rotation="-14.399,-0.000,0.000", Scale="1,1,0.5") -> (Pose) {
    alphaInputType=Bool
    rotationMode=BMM_Additive
    rotationSpace=BCS_BoneSpace
    scaleMode=BMM_Additive
    scaleSpace=BCS_BoneSpace
    translationMode=BMM_Additive
    translationSpace=BCS_BoneSpace
    alphaBoolBlend {blendInTime=0.2, blendOutTime=0.2}
    boneToModify {boneName=dyn_hair_4_c}
    rotation=<rot -14.399 -0.0001 0>
    scale {z=0.5}
   }
   n33 = "Transform (Modify) Bone - Bone: dyn_hair_4_l"(ComponentPose=@n32, bAlphaBoolEnabled=True, Translation="0,0,2.5", Rotation="10,0,0", Scale="1,1,0.4") {
    alphaInputType=Bool
    rotationMode=BMM_Additive
    rotationSpace=BCS_BoneSpace
    scaleMode=BMM_Additive
    scaleSpace=BCS_BoneSpace
    translationMode=BMM_Additive
    translationSpace=BCS_BoneSpace
    alphaBoolBlend {blendInTime=0.2, blendOutTime=0.2}
    boneToModify {boneName=dyn_hair_4_l}
    rotation {pitch=10}
    scale {z=0.4}
    translation {z=2.5}
   }
   Output Pose(Result=Component To Local(ComponentPose=@n33)) {layerGroup=None}
   n34 = Control Rig!(
    Source=Copy Pose From Mesh() {bUseAttachedParent=true}
    Alpha
    isSkydiving_CR:bool=Get bIsSkyDiving()
    isCrouching_CR:bool=Get bIsCrouching()
   ) -> (Pose) {controlRigAssetReference {blueprintRigClass=&/F_MED_ForestJeanWinkHat_FaceAcc_CtrlRig.F_MED_ForestJeanWinkHat_FaceAcc_CtrlRig_C}}
   Save cached pose 'preSimCache'(Pose=@n34) {cacheName=preSimCache}
  ]}
 EventGraph {nodes [
   Event BlueprintUpdateAnimation()
   > Sequence()
   > n1 = "Gravity Override State Function (Parameter Asset) (Nativized)"(
    CharacterPartAnimInstance=Self-Reference()
    GameTimeInSeconds=GetGameTimeInSeconds()
    Parameters=/BRCosmetics/Characters/Player/Female/Medium/Bodies/F_MED_ForestJeanWink/Meshes/Parts/F_MED_ForestJeanWinkFA_Parameters
   )
   > Set GravityOverride(GravityOverride=@n1)
   n2 = Get bRigidBodyEnabled()
   n3 = CheckForBackbling()
   > Set isWearingBackbling / Cape / NoRBAN(isWearingBackbling / Cape / NoRBAN:bool=OR Boolean!(A=@n3, B=@n3.bIsBackblingACape, C:bool=@n3.bIsBackblingLong, D:bool=NOT Boolean(A=@n2)) -> (ReturnValue)) -> (Output_Get:bool)
  ]}
]
variables [
 GravityOverride {type=Vector}
 isWearingBackbling / Cape / NoRBAN {type=boolean}
]
